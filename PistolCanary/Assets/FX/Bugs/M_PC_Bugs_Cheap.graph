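# Default Lit bug material for Niagara mesh particles, named as the cheap variant, using bug diffuse and emissive textures.
# Gradient 3D-texture noise, sampled from world X and Z (X repeated for Y) and scrolled along Z by time, is scaled by WPO_Amount for vertex wobble.
# A feature-level switch zeroes that noise on ES2 and ES3.1.
# The emissive mask's red channel is tinted by particle color.
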Material /PistolCanary/Assets/FX/Bugs/M_PC_Bugs_Cheap
graph {
 properties {bAutomaticallySetUsageInEditor=false, bUsedWithNiagaraMeshParticles=true, usedShadingModels=Default Lit}
 nodes [
  n1 = MaterialFunctionCall(Float3=WorldPosition().XYZ) {materialFunction=&BreakOutFloat3Components}
  n2 = Multiply(A=Time(), B=ScalarParameter() {defaultValue=250, parameterName=NoiseSpeed}) {constB=100}
  n3 = Noise!(
   World Position=MaterialFunctionCall(R=@n1.R, G=@n1.R, B=Add(A=@n1.B, B=@n2)) {materialFunction=&AppendMany}.RGB
   FilterWidth
  ) {bTurbulence=false, levelScale=3, levels=4, noiseFunction=NOISEFUNCTION_GradientTex3D, outputMin=-1.5, scale=0.002}
  n4 = Multiply(A=@n3, B=ScalarParameter() {defaultValue=40, parameterName=WPO_Amount})
  n5 = Constant()
  FeatureLevelSwitch(Default=@n4, ES2=@n5, ES3_1=@n5, SM5=@n4, SM6=@n4)
  Multiply(A=TextureSample() {texture=&/T_Bug_Emissive}.R, B=ParticleColor().RGB)
  TextureSample() {texture=&/T_Bug_Diffuse}
 ]
}
