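# Niagara system for ambient bugs, using one emitter with sprite and mesh renderers.
# Particles spawn in an owner-scaled sphere and circle randomized centers and axes, with meshes facing along the orbit.
# Mesh size and alpha fade over lifetime, scaled by a distance-fraction curve that fades both near and far.
# Emitter culling sleeps and clears particles outside that distance band; curl noise and waveform drag are present but disabled.
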NiagaraSystem /PistolCanary/Assets/FX/Bugs/NS_PC_Bugs_Ambient
dependencies {usedDataInterfaces [&NiagaraDataInterfaceCurve], usedDynamicInputs [], usedModules [&SystemState]}
emitters {Bugs {
  emitterUpdateScript [
   &EmitterState(
    Enable Distance Culling=true
    Max Distance Response=::SleepAndClearParticles
    MaxDistance=9000
    Min Distance Response=::SleepAndClearParticles
    MinDistance=1000
    Module.bUseMaxDistance=true
    Module.bUseMinDistance=true
    Reset Age On Awaken=true
    Scalability Mode=::NewEnumerator1
   )
   &SpawnRate(Use Spawn Probability=false)
  ]
  particleSpawnScript [
   &InitializeParticle(
    Color=<rgba 5 0.1847 2.7594 0.01>
    Lifetime Max=8
    Lifetime Min=5
    Lifetime Mode=::NewEnumerator1
    Mesh Scale Mode=::NewEnumerator1
    Mesh Uniform Scale Max=0.7
    Mesh Uniform Scale Min=0.5
    Sprite Rotation Mode=::NewEnumerator2
    Sprite Size Mode=::NewEnumerator3
    Sprite UV Mode=::NewEnumerator3
    Uniform Sprite Size=150
    UsePositionOffset=false
    Write Material Random=true
   )
   &ShapeLocation(MaskSpawnGroup=false, Override Seed=false, Sphere Radius=250, UseOwnerScale=true)
  ]
  particleUpdateScript [
   &ParticleState()
   /SetVariables(Particles.DistanceCurve:NiagaraFloat=&FloatFromCurve(
     FloatCurve=DataInterface() {propertyValues {Curve {
        defaultValue=1
        keys [
         <at 0.15 0 RCIM_Cubic>
         {arriveTangent=11.1894, tangentMode=RCTM_User, time=0.3, value=1}
         {interpMode=RCIM_Cubic, leaveTangent=-18.3558, tangentMode=RCTM_Break, time=0.9, value=1}
         <at 1 0 RCIM_Cubic>
        ]
       }}}
     CurveIndex=$Emitter.Scalability.DistanceFraction
    )) {isSetParameters=true}
   &ScaleMeshSize(Scale Factor=&VectorFromFloat(Value=&FloatFromCurve(
      FloatCurve=DataInterface() {propertyValues {Curve {
         defaultValue=1
         keys [
          {interpMode=RCIM_Cubic, tangentMode=RCTM_User, time=0, value=0}
          {arriveTangent=0.003, interpMode=RCIM_Cubic, leaveTangent=0.003, tangentMode=RCTM_User, time=0.2, value=1}
          {interpMode=RCIM_Cubic, tangentMode=RCTM_User, time=0.8, value=1}
          {tangentMode=RCTM_User, time=1, value=0}
         ]
        }}}
      Scale Curve=$Particles.DistanceCurve
     )))
   &ScaleColor(
    Scale Alpha=&FloatFromCurve(
     FloatCurve=DataInterface() {propertyValues {Curve {
        defaultValue=1
        keys [
         {interpMode=RCIM_Cubic, tangentMode=RCTM_User, time=0, value=0}
         {interpMode=RCIM_Cubic, tangentMode=RCTM_User, time=0.5, value=1}
         {tangentMode=RCTM_User, time=1, value=0}
        ]
       }}}
     Scale Curve=$Particles.DistanceCurve
    )
    ScaleA=true
    ScaleRGB=true
   )
   &RotateAroundPoint(
    Derive the Rotation Axis=false
    Radius:NiagaraFloat=&RandomRangeFloat(Minimum=75, Maximum=250)
    Rotation Axis:Vector3f=&RandomRangeVector(Minimum=<-1 -1 1>, Maximum=<1 1 1>)
    Rotation Center:NiagaraPosition=&RandomRangeVector(Minimum=<0 0 0>, Maximum=<0 0 150>)
    Rotation Phase:NiagaraFloat=&RandomRangeFloat(Minimum=0, Maximum=1)
    Rotation Rate:NiagaraFloat=&RandomRangeFloat(Minimum=0.2, Maximum=0.23)
   )
   &UpdateMeshOrientation(
    Facing Direction=$Output.RotateAroundPoint.LocalForwardVector
    Facing Interpolation=false
    Orientation Method=::NewEnumerator1
    Override Quat=false
    TransformFacing=false
    TransformUp=false
   )
   &CurlNoiseForce(Bias Noise=false, Noise Frequency=10, Noise Strength=50, Pan Noise=false) {enabled=false}
   &Drag(
    Drag=&Waveform("Waveform [1]"=::NewEnumerator8, "[1] Frequency"=2, "[1] Amplitude Min/Max"=<0 3>)
    Use Linear Drag=true
    Use Rotational Drag=true
   ) {enabled=false}
   &SolveForcesAndVelocity(Clamp Velocity=false, Limit Acceleration=false)
  ]
  rendererClasses [&NiagaraSpriteRendererProperties, &NiagaraMeshRendererProperties]
  renderers [
   {rendererClass=&NiagaraSpriteRendererProperties}
   {rendererClass=&NiagaraMeshRendererProperties, rendererIndex=1}
  ]
 }}
summary {
 systemName=NS_PC_Bugs_Ambient
 emitters [{emitterName=Bugs, rendererClasses [&NiagaraSpriteRendererProperties, &NiagaraMeshRendererProperties]}]
 userVariables []
}
